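# Physics asset for the M_MED_Banana_Summer player body cosmetic, covering its shorts and a bow with loops and laces.
# Each thigh carries four shorts panels (front, back, in, out), each angle-limited against a tiny base box that links to the thigh with free rotation.
# The bow links through its base to the pelvis with free rotation, carrying angle-limited loops and two-link lace chains.
# Capsules shape the pelvis and thighs.
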PhysicsAsset /BRCosmetics/Characters/Player/Male/Medium/Bodies/M_MED_Banana_Summer/Meshes/M_MED_Banana_Summer_Physics
bodies {
 dyn_bowMain [{shapeType=Box, extentX=1.1, extentY=1.1, extentZ=1.1, center=<0.1688 0.4361 -0.322>, rotation=<rot 0 -90.3115 -83.9286>}]
 dyn_bowMain_Base [{shapeType=Box, extentX=1.1, extentY=1.1, extentZ=1.1, rotation=<rot 0 -90 0>}]
 dyn_lace_01_l [{shapeType=Capsule, length=1, radius=0.7585, center=<2 0 -0.5>, rotation=<rot 0 0 90>}]
 dyn_lace_01_r [{shapeType=Capsule, length=1, radius=0.7453, center=<2 0 0.5>, rotation=<rot 0 0 90>}]
 dyn_lace_02_l [{shapeType=Capsule, length=1, radius=0.736, center=<2 0 -0.5>, rotation=<rot 0 0 90>}]
 dyn_lace_02_r [{shapeType=Capsule, length=1, radius=0.7427, center=<2 0 0.5>, rotation=<rot 0 0 90>}]
 dyn_loop_l [{shapeType=Capsule, length=3, radius=1.5, shapeName=dyn_loop_l_capsule, center=<2.9503 0 -0.282>, rotation=<rot 80 -0.0001 -0.0002>}]
 dyn_loop_r [{shapeType=Capsule, length=2, radius=1.5, shapeName=dyn_loop_r_capsule, center=<2 0 0>, rotation=<rot 90 0 0>}]
 dyn_shortsBack_Base_l [{shapeType=Box, extentX=1.01, extentY=1.01, extentZ=1.01, rotation=<rot 0 -90 0>}]
 dyn_shortsBack_Base_r [{shapeType=Box, extentX=1.01, extentY=1.01, extentZ=1.01, rotation=<rot 0 -90 0>}]
 dyn_shortsBack_l [{shapeType=Box, extentX=9.5264, extentY=3.3088, extentZ=16.2327, center=<-4.3838 -0.401 0>}]
 dyn_shortsBack_r [{shapeType=Box, extentX=9.5264, extentY=3.3088, extentZ=16.2327, center=<4.3838 0.401 0>, rotation=<rot 0 0 180>}]
 dyn_shortsFront_Base_l [{shapeType=Box, extentX=1.01, extentY=1.01, extentZ=1.01, rotation=<rot 0 -90 0>}]
 dyn_shortsFront_Base_r [{shapeType=Box, extentX=1.01, extentY=1.01, extentZ=1.01, rotation=<rot 0 -90 0>}]
 dyn_shortsFront_l [{shapeType=Box, extentX=9.441, extentY=4.0147, extentZ=14.2442, center=<-3.7921 0 0>}]
 dyn_shortsFront_r [{shapeType=Box, extentX=9.441, extentY=4.0147, extentZ=14.2442, center=<3.7921 0 0>, rotation=<rot 0 0 180>}]
 dyn_shortsIn_Base_l [{shapeType=Box, extentX=1.01, extentY=1.01, extentZ=1.01, rotation=<rot 0 -90 0>}]
 dyn_shortsIn_Base_r [{shapeType=Box, extentX=1.01, extentY=1.01, extentZ=1.01, rotation=<rot 0 -90 0>}]
 dyn_shortsIn_l [{shapeType=Box, extentX=8.3704, extentY=3.2919, extentZ=14.2093, center=<-3.0126 0.3212 0.0001>, rotation=<rot 0 -9.5862 0>}]
 dyn_shortsIn_r [{shapeType=Box, extentX=8.3704, extentY=3.2919, extentZ=14.2093, center=<3.0126 -0.3212 -0.0001>, rotation=<rot 0 -9.5862 -180>}]
 dyn_shortsOut_Base_l [{shapeType=Box, extentX=1.01, extentY=1.01, extentZ=1.01, rotation=<rot 0 -90 0>}]
 dyn_shortsOut_Base_r [{shapeType=Box, extentX=1.01, extentY=1.01, extentZ=1.01, rotation=<rot 0 -90 0>}]
 dyn_shortsOut_l [{shapeType=Box, extentX=9.317, extentY=3.484, extentZ=18.1515, center=<-3.9511 0.0001 0>, rotation=<rot -4.0868 0 0>}]
 dyn_shortsOut_r [{shapeType=Box, extentX=9.317, extentY=3.484, extentZ=18.1515, center=<3.9511 -0.0001 0>, rotation=<rot -4.0868 0 180>}]
 pelvis [
  {shapeType=Capsule, length=6, radius=16, shapeName=pelvis_capsule, center=<-6.3741 -0.3349 0>}
  {shapeType=Capsule, length=20, radius=4, shapeName=pelvis_capsule, center=<-7.7965 12.1431 0>}
 ]
 thigh_l [{shapeType=Capsule, length=20, radius=10, shapeName=thigh_l_capsule, center=<-20 0 0>, rotation=<rot 90 0 0>}]
 thigh_r [{shapeType=Capsule, length=20, radius=10, shapeName=thigh_r_capsule, center=<20 0 0>, rotation=<rot 90 0 0>}]
}
constraints [
 {bone1Name=dyn_shortsFront_l, bone2Name=dyn_shortsFront_Base_l, swing1LimitDegrees=30, swing2LimitDegrees=10, twistLimitDegrees=8}
 {bone1Name=dyn_shortsOut_l, bone2Name=dyn_shortsOut_Base_l, swing1LimitDegrees=20, swing2LimitDegrees=10, twistLimitDegrees=8}
 {bone1Name=dyn_shortsIn_l, bone2Name=dyn_shortsIn_Base_l, swing1LimitDegrees=16, swing2LimitDegrees=12, twistLimitDegrees=12}
 {bone1Name=dyn_shortsBack_l, bone2Name=dyn_shortsBack_Base_l, swing1LimitDegrees=16, swing2LimitDegrees=16, twistLimitDegrees=12}
 {bone1Name=dyn_shortsFront_r, bone2Name=dyn_shortsFront_Base_r, swing1LimitDegrees=30, swing2LimitDegrees=10, twistLimitDegrees=8}
 {bone1Name=dyn_shortsOut_r, bone2Name=dyn_shortsOut_Base_r, swing1LimitDegrees=20, swing2LimitDegrees=10, twistLimitDegrees=8}
 {bone1Name=dyn_shortsBack_r, bone2Name=dyn_shortsBack_Base_r, swing1LimitDegrees=16, swing2LimitDegrees=16, twistLimitDegrees=12}
 {bone1Name=dyn_shortsIn_r, bone2Name=dyn_shortsIn_Base_r, swing1LimitDegrees=16, swing2LimitDegrees=12, twistLimitDegrees=12}
 {
  bone1Name=dyn_bowMain
  bone2Name=dyn_bowMain_Base
  swing1LimitDegrees=45
  swing1Motion=Free
  swing2LimitDegrees=45
  swing2Motion=Free
  twistLimitDegrees=45
  twistMotion=Free
 }
 {bone1Name=dyn_loop_r, bone2Name=dyn_bowMain, swing1LimitDegrees=25, swing2LimitDegrees=25, twistLimitDegrees=2}
 {bone1Name=dyn_loop_l, bone2Name=dyn_bowMain, swing1LimitDegrees=25, swing2LimitDegrees=25, twistLimitDegrees=2}
 {bone1Name=dyn_lace_01_r, bone2Name=dyn_bowMain, swing1LimitDegrees=15, swing2LimitDegrees=15, twistLimitDegrees=2}
 {bone1Name=dyn_lace_02_r, bone2Name=dyn_lace_01_r, swing1LimitDegrees=15, swing2LimitDegrees=15, twistLimitDegrees=2}
 {bone1Name=dyn_lace_01_l, bone2Name=dyn_bowMain, swing1LimitDegrees=15, swing2LimitDegrees=15, twistLimitDegrees=2}
 {bone1Name=dyn_lace_02_l, bone2Name=dyn_lace_01_l, swing1LimitDegrees=15, swing2LimitDegrees=15, twistLimitDegrees=2}
 {
  bone1Name=dyn_shortsFront_Base_l
  bone2Name=thigh_l
  swing1LimitDegrees=45
  swing1Motion=Free
  swing2LimitDegrees=45
  swing2Motion=Free
  twistLimitDegrees=45
  twistMotion=Free
 }
 {
  bone1Name=dyn_shortsOut_Base_l
  bone2Name=thigh_l
  swing1LimitDegrees=45
  swing1Motion=Free
  swing2LimitDegrees=45
  swing2Motion=Free
  twistLimitDegrees=45
  twistMotion=Free
 }
 {
  bone1Name=dyn_shortsIn_Base_l
  bone2Name=thigh_l
  swing1LimitDegrees=45
  swing1Motion=Free
  swing2LimitDegrees=45
  swing2Motion=Free
  twistLimitDegrees=45
  twistMotion=Free
 }
 {
  bone1Name=dyn_shortsBack_Base_l
  bone2Name=thigh_l
  swing1LimitDegrees=45
  swing1Motion=Free
  swing2LimitDegrees=45
  swing2Motion=Free
  twistLimitDegrees=45
  twistMotion=Free
 }
 {
  bone1Name=dyn_shortsBack_Base_r
  bone2Name=thigh_r
  swing1LimitDegrees=45
  swing1Motion=Free
  swing2LimitDegrees=45
  swing2Motion=Free
  twistLimitDegrees=45
  twistMotion=Free
 }
 {
  bone1Name=dyn_shortsIn_Base_r
  bone2Name=thigh_r
  swing1LimitDegrees=45
  swing1Motion=Free
  swing2LimitDegrees=45
  swing2Motion=Free
  twistLimitDegrees=45
  twistMotion=Free
 }
 {
  bone1Name=dyn_shortsOut_Base_r
  bone2Name=thigh_r
  swing1LimitDegrees=45
  swing1Motion=Free
  swing2LimitDegrees=45
  swing2Motion=Free
  twistLimitDegrees=45
  twistMotion=Free
 }
 {
  bone1Name=dyn_shortsFront_Base_r
  bone2Name=thigh_r
  swing1LimitDegrees=45
  swing1Motion=Free
  swing2LimitDegrees=45
  swing2Motion=Free
  twistLimitDegrees=45
  twistMotion=Free
 }
 {
  bone1Name=dyn_bowMain_Base
  bone2Name=pelvis
  swing1LimitDegrees=45
  swing1Motion=Free
  swing2LimitDegrees=45
  swing2Motion=Free
  twistLimitDegrees=45
  twistMotion=Free
 }
 {
  bone1Name=thigh_l
  bone2Name=pelvis
  swing1LimitDegrees=45
  swing1Motion=Free
  swing2LimitDegrees=45
  swing2Motion=Free
  twistLimitDegrees=45
  twistMotion=Free
 }
 {
  bone1Name=thigh_r
  bone2Name=pelvis
  swing1LimitDegrees=45
  swing1Motion=Free
  swing2LimitDegrees=45
  swing2Motion=Free
  twistLimitDegrees=45
  twistMotion=Free
 }
]
