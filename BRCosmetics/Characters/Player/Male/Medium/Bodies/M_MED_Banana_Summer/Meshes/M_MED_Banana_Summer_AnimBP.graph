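# Animation Blueprint for the M_MED_Banana_Summer player body cosmetic, adding procedural secondary motion to the pose copied from its attached parent mesh.
# The copied pose and curves feed a Control Rig, then an additive bow-bone offset, then RigidBody simulation in root-bone space.
# Additive rotations on both lace bones follow the simulation.
# Each update turns bIsCrouching into a hard 0/1 alpha that gates the bow and lace adjustments.
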AnimBlueprint /BRCosmetics/Characters/Player/Male/Medium/Bodies/M_MED_Banana_Summer/Meshes/M_MED_Banana_Summer_AnimBP
target_skeleton=&/M_MED_Banana_Summer_Skeleton
graphs [
 AnimGraph {nodes [
   n1 = Control Rig(Source=Copy Pose From Mesh() {bCopyCurves=true, bUseAttachedParent=true}) {
    alphaBoolBlend {blendOption=Linear}
    controlRigAssetReference {blueprintRigClass=&/M_MED_Banana_Summer_CtrlRig.M_MED_Banana_Summer_CtrlRig_C}
   }
   n2 = "Transform (Modify) Bone - Bone: dyn_bowMain_Base"(
    ComponentPose=Local To Component(LocalPose=@n1)
    Alpha=Get Is Crouching Alpha() -> (Is Crouching Alpha:double)
    Translation="0,1,1"
    Rotation="0,0,0"
    Scale="1,1,1"
   ) {
    rotationMode=BMM_Additive
    rotationSpace=BCS_BoneSpace
    scaleSpace=BCS_BoneSpace
    translationMode=BMM_Additive
    translationSpace=BCS_BoneSpace
    alphaBoolBlend {blendOption=Linear}
    boneToModify {boneName=dyn_bowMain_Base}
    translation {y=1, z=1}
   }
   n3 = RigidBody(ComponentPose=@n2) {
    bClampLinearTranslationLimitToRefPose=true
    bOverrideWorldGravity=true
    overridePhysicsAsset=&/M_MED_Banana_Summer_Physics
    simulationSpace=BaseBoneSpace
    alphaBoolBlend {blendOption=Linear}
    baseBoneRef {boneName=root}
    componentAppliedLinearAccClamp=<4000 4000 4000>
    componentLinearAccScale=<0.8 0.8 0.8>
    componentLinearVelScale=<0.35 35 0.35>
    overrideWorldGravity {z=-980}
   }
   n4 = Get Is Crouching Alpha()
   n5 = "Transform (Modify) Bone - Bone: dyn_lace_01_r"(ComponentPose=@n3, Alpha=@n4, Translation="0,0,0", Rotation="-10,0,0", Scale="1,1,1") {
    rotationMode=BMM_Additive
    rotationSpace=BCS_BoneSpace
    scaleSpace=BCS_BoneSpace
    translationMode=BMM_Additive
    translationSpace=BCS_BoneSpace
    alphaBoolBlend {blendOption=Linear}
    boneToModify {boneName=dyn_lace_01_r}
    rotation {pitch=-10}
   }
   n6 = "Transform (Modify) Bone - Bone: dyn_lace_01_l"(ComponentPose=@n5, Alpha=@n4, Translation="0,0,0", Rotation="10,0,0", Scale="1,1,1") {
    rotationMode=BMM_Additive
    rotationSpace=BCS_BoneSpace
    scaleSpace=BCS_BoneSpace
    translationMode=BMM_Additive
    translationSpace=BCS_BoneSpace
    alphaBoolBlend {blendOption=Linear}
    boneToModify {boneName=dyn_lace_01_l}
    rotation {pitch=10}
   }
   Output Pose(Result=Component To Local(ComponentPose=@n6)) {layerGroup=None}
  ]}
 EventGraph {nodes [
   Event BlueprintUpdateAnimation()
   > Set Is Crouching Alpha(Is Crouching Alpha:double="To Float (Boolean)"(InBool=Get bIsCrouching())) -> (Output_Get:double)
  ]}
]
variables [Is Crouching Alpha {type=number}]
